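# Animation Blueprint for the TalonHime backpack that gives its dyn_spine chain secondary physics motion.
# It copies the character's pelvis and spine_03 transforms, then runs chained AnimDynamics, mostly with spherical limits around body and head.
# Spine-versus-pelvis rotation speed drives an additive bend on the last three bones, suppressed while jumping, falling or sprinting.
# Character state also tunes gravity, inertia (zeroed when crouching or skydiving) and front-end blend weight.
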AnimBlueprint /BRCosmetics/Accessories/FORT_Backpacks/Backpack_F_MED_TalonHime/Meshes/F_MED_TalonHime_Pack_AnimBP
target_skeleton=/Game/Accessories/FORT_Backpacks/Backpack_F_MED_TalonHime/Meshes/F_MED_TalonHime_Pack_Skeleton
graphs [
 AnimGraph {nodes [
   n1 = "Transform (Modify) Bone - Bone: pelvis"(
    ComponentPose=Local To Component(LocalPose=Input Pose())
    Alpha=1
    Translation=Get Pelvis Translation() -> (Pelvis Translation:Vector)
    Rotation="Get Pelvis  Rotation"() -> ("Pelvis  Rotation":Rotator)
    Scale="1,1,1"
   ) {
    rotationMode=BMM_Replace
    rotationSpace=BCS_WorldSpace
    translationMode=BMM_Replace
    translationSpace=BCS_WorldSpace
    alphaBoolBlend {blendOption=Linear}
    boneToModify {boneName=pelvis}
   }
   n2 = "Transform (Modify) Bone - Bone: spine_03"(
    ComponentPose=@n1
    Alpha=1
    Translation=Get Spine_03 Translation() -> (Spine_03 Translation:Vector)
    Rotation=Get Spine_03 Rotation() -> (Spine_03 Rotation:Rotator)
    Scale="1,1,1"
   ) {
    rotationMode=BMM_Replace
    rotationSpace=BCS_WorldSpace
    translationMode=BMM_Replace
    translationSpace=BCS_WorldSpace
    alphaBoolBlend {blendOption=Linear}
    boneToModify {boneName=spine_03}
   }
   n3 = Anim Dynamics - Bone: dyn_spine_01!(
    ComponentPose=@n2
    Alpha=1
    GravityOverride:Vector="0,0,-980"
    ComponentLinearAccScale:Vector="0.5,0.3,0.2"
    ComponentLinearVelScale:Vector="0.5,0.3,0.2"
    ExternalForce="0,0,0"
   ) -> (Pose) {
    angularDampingOverride=0.9999
    angularSpringConstant=350
    bAngularSpring=true
    bOverrideAngularDamping=true
    bUseGravityOverride=true
    bUseSphericalLimits=true
    simulationSpace=RootRelative
    alphaBoolBlend {blendOption=Linear}
    boundBone {boneName=dyn_spine_01}
    componentAppliedLinearAccClamp=<2000 2000 2000>
    componentLinearAccScale=<0.3 0.3 0.3>
    componentLinearVelScale=<0.3 0.3 0.3>
    gravityOverride {z=-700}
    physicsBodyDefinitions [{
      boundBone {boneName=dyn_spine_01}
      boxExtents=<20 20 20>
      constraintSetup {angularLimitsMax=<0 0 5>, angularLimitsMin=<0 0 -5>, angularTarget=<0 0 0>}
      localJointOffset=<0.5 0 -0.5>
     }]
   }
   n4 = Get Alpha_0() -> (Alpha_0:float)
   n5 = Get Gravity Override()
   n6 = Get Component Linear Acc Scale_2() -> (Component Linear Acc Scale_2:Vector)
   n7 = Get Component Linear Vel Scale_2() -> (Component Linear Vel Scale_2:Vector)
   n8 = Anim Dynamics - Bone: dyn_spine_02!(
    ComponentPose=@n3
    Alpha=@n4
    GravityOverride:Vector=@n5
    ComponentLinearAccScale:Vector=@n6
    ComponentLinearVelScale:Vector=@n7
    ExternalForce="0,0,0"
   ) -> (Pose) {
    angularDampingOverride=0.999
    angularSpringConstant=350
    bAngularSpring=true
    bOverrideAngularDamping=true
    bUseGravityOverride=true
    bUseSphericalLimits=true
    simulationSpace=RootRelative
    alphaBoolBlend {blendOption=Linear}
    boundBone {boneName=dyn_spine_02}
    componentAppliedLinearAccClamp=<2000 2000 2000>
    componentLinearAccScale=<0.3 0.3 0.3>
    componentLinearVelScale=<0.3 0.3 0.3>
    gravityOverride {z=-700}
    physicsBodyDefinitions [{
      boundBone {boneName=dyn_spine_02}
      boxExtents=<8 8 8>
      constraintSetup {angularLimitsMax=<0 0 10>, angularLimitsMin=<0 -2 -10>, angularTarget=<0 0 0>}
      localJointOffset=<0.5 0 0>
     }]
    sphericalLimits [
     {limitRadius=16, drivingBone {boneName=pelvis}, sphereLocalOffset=<0 0 0>}
     {limitRadius=15.492, drivingBone {boneName=dyn_head}, sphereLocalOffset=<-1.2288 -0.1098 14.9879>}
     {limitRadius=11.0771, drivingBone {boneName=spine_03}, sphereLocalOffset=<0 0 0>}
    ]
   }
   n9 = Anim Dynamics - Bone: dyn_spine_03!(
    ComponentPose=@n8
    Alpha=@n4
    GravityOverride:Vector=@n5
    ComponentLinearAccScale:Vector=@n6
    ComponentLinearVelScale:Vector=@n7
    ExternalForce="0,0,0"
   ) -> (Pose) {
    angularDampingOverride=0.999
    angularSpringConstant=350
    bAngularSpring=true
    bOverrideAngularDamping=true
    bUseGravityOverride=true
    bUseSphericalLimits=true
    simulationSpace=RootRelative
    alphaBoolBlend {blendOption=Linear}
    boundBone {boneName=dyn_spine_03}
    componentAppliedLinearAccClamp=<2000 2000 2000>
    componentLinearAccScale=<0.3 0.3 0.3>
    componentLinearVelScale=<0.3 0.3 0.3>
    gravityOverride {z=-700}
    physicsBodyDefinitions [{
      boundBone {boneName=dyn_spine_03}
      boxExtents=<8 8 8>
      constraintSetup {angularLimitsMax=<0 0 15>, angularLimitsMin=<0 -5 -15>, angularTarget=<0 0 0>}
      localJointOffset=<0.5 0 0>
     }]
    sphericalLimits [
     {limitRadius=16, drivingBone {boneName=pelvis}, sphereLocalOffset=<0 0 0>}
     {limitRadius=15.492, drivingBone {boneName=dyn_head}, sphereLocalOffset=<-1.2288 -0.1098 14.9879>}
     {limitRadius=11.0771, drivingBone {boneName=spine_03}, sphereLocalOffset=<0 0 0>}
    ]
   }
   n10 = Anim Dynamics - Bone: dyn_spine_04!(
    ComponentPose=@n9
    Alpha=@n4
    GravityOverride:Vector=@n5
    ComponentLinearAccScale:Vector=@n6
    ComponentLinearVelScale:Vector=@n7
    ExternalForce="0,0,0"
   ) -> (Pose) {
    angularDampingOverride=0.999
    angularSpringConstant=350
    bAngularSpring=true
    bOverrideAngularDamping=true
    bUseGravityOverride=true
    bUseSphericalLimits=true
    simulationSpace=RootRelative
    alphaBoolBlend {blendOption=Linear}
    boundBone {boneName=dyn_spine_04}
    componentAppliedLinearAccClamp=<2000 2000 2000>
    componentLinearAccScale=<0.3 0.3 0.3>
    componentLinearVelScale=<0.3 0.3 0.3>
    gravityOverride {z=-700}
    physicsBodyDefinitions [{
      boundBone {boneName=dyn_spine_04}
      boxExtents=<7.7 7.7 7.7>
      constraintSetup {angularLimitsMax=<0 0 20>, angularLimitsMin=<0 -12 -20>, angularTarget=<0 0 0>}
      localJointOffset=<0.5 0 0>
     }]
    sphericalLimits [
     {limitRadius=16, drivingBone {boneName=pelvis}, sphereLocalOffset=<0 0 0>}
     {limitRadius=15.492, drivingBone {boneName=dyn_head}, sphereLocalOffset=<-1.2288 -0.1098 14.9879>}
     {limitRadius=11.0771, drivingBone {boneName=spine_03}, sphereLocalOffset=<0 0 0>}
    ]
   }
   n11 = Anim Dynamics - Bone: dyn_spine_05!(
    ComponentPose=@n10
    Alpha=@n4
    GravityOverride:Vector=@n5
    ComponentLinearAccScale:Vector=@n6
    ComponentLinearVelScale:Vector=@n7
    ExternalForce="0,0,0"
   ) -> (Pose) {
    angularDampingOverride=0.999
    angularSpringConstant=350
    bAngularSpring=true
    bOverrideAngularDamping=true
    bUseGravityOverride=true
    bUseSphericalLimits=true
    simulationSpace=RootRelative
    alphaBoolBlend {blendOption=Linear}
    boundBone {boneName=dyn_spine_05}
    componentAppliedLinearAccClamp=<2000 2000 2000>
    componentLinearAccScale=<0.3 0.3 0.3>
    componentLinearVelScale=<0.3 0.3 0.3>
    gravityOverride {z=-700}
    physicsBodyDefinitions [{
      boundBone {boneName=dyn_spine_05}
      boxExtents=<7.5 7.5 7.5>
      constraintSetup {angularLimitsMax=<0 0 25>, angularLimitsMin=<0 -15 -25>, angularTarget=<0 0 0>}
      localJointOffset=<0.5 0 0>
     }]
    sphericalLimits [
     {limitRadius=16, drivingBone {boneName=pelvis}, sphereLocalOffset=<0 0 0>}
     {limitRadius=15.492, drivingBone {boneName=dyn_head}, sphereLocalOffset=<-1.2288 -0.1098 14.9879>}
     {limitRadius=11.0771, drivingBone {boneName=spine_03}, sphereLocalOffset=<0 0 0>}
    ]
   }
   n12 = Anim Dynamics - Bone: dyn_spine_06!(
    ComponentPose=@n11
    Alpha=@n4
    GravityOverride:Vector=@n5
    ComponentLinearAccScale:Vector=@n6
    ComponentLinearVelScale:Vector=@n7
    ExternalForce="0,0,0"
   ) -> (Pose) {
    angularDampingOverride=0.999
    angularSpringConstant=350
    bAngularSpring=true
    bOverrideAngularDamping=true
    bUseGravityOverride=true
    bUseSphericalLimits=true
    simulationSpace=RootRelative
    alphaBoolBlend {blendOption=Linear}
    boundBone {boneName=dyn_spine_06}
    componentAppliedLinearAccClamp=<2000 2000 2000>
    componentLinearAccScale=<0.3 0.3 0.3>
    componentLinearVelScale=<0.3 0.3 0.3>
    gravityOverride {z=-700}
    physicsBodyDefinitions [{
      boundBone {boneName=dyn_spine_06}
      boxExtents=<7.5 7.5 7.5>
      constraintSetup {angularLimitsMax=<0 0 20>, angularLimitsMin=<0 -15 -20>, angularTarget=<0 0 0>}
      localJointOffset=<0.5 0 0>
     }]
    sphericalLimits [
     {limitRadius=16, drivingBone {boneName=pelvis}, sphereLocalOffset=<0 0 0>}
     {limitRadius=15.492, drivingBone {boneName=dyn_head}, sphereLocalOffset=<-1.2288 -0.1098 14.9879>}
     {limitRadius=11.0771, drivingBone {boneName=spine_03}, sphereLocalOffset=<0 0 0>}
    ]
   }
   n13 = Get AlphaRotationHistory() -> (AlphaRotationHistory:double)
   n14 = "Transform (Modify) Bone - Bone: dyn_spine_06"(ComponentPose=@n12, Alpha=@n13, Translation="0,0,0", Rotation="15,0,0", Scale="1,1,1") {
    rotationMode=BMM_Additive
    rotationSpace=BCS_BoneSpace
    alphaBoolBlend {blendOption=Linear}
    boneToModify {boneName=dyn_spine_06}
    rotation {pitch=15}
   }
   n15 = "Transform (Modify) Bone - Bone: dyn_spine_05"(ComponentPose=@n14, Alpha=@n13, Translation="0,0,0", Rotation="15,0,0", Scale="1,1,1") {
    rotationMode=BMM_Additive
    rotationSpace=BCS_BoneSpace
    alphaBoolBlend {blendOption=Linear}
    boneToModify {boneName=dyn_spine_05}
    rotation {pitch=15}
   }
   n16 = "Transform (Modify) Bone - Bone: dyn_spine_04"(ComponentPose=@n15, Alpha=@n13, Translation="0,0,0", Rotation="15,0,0", Scale="1,1,1") {
    rotationMode=BMM_Additive
    rotationSpace=BCS_BoneSpace
    alphaBoolBlend {blendOption=Linear}
    boneToModify {boneName=dyn_spine_04}
    rotation {pitch=15}
   }
   n17 = "Transform (Modify) Bone - Bone: pack"(ComponentPose=@n16, Translation="0,3,0", Rotation="0,0,0") {translationMode=BMM_Additive, alphaBoolBlend {blendOption=Linear}, boneToModify {boneName=pack}, translation {y=3}}
   Output Pose(Result=Component To Local(ComponentPose=@n17)) {layerGroup=None}
  ]}
 EventGraph {nodes [
   n1 = Event BlueprintUpdateAnimation()
   > n2 = Sequence()
   n3 = Get SkelMesh()
   > Is Valid(InputObject=@n3)
   n4 = BreakTransform(InTransform=GetSocketTransform(self=@n3, InSocketName=pelvis, TransformSpace=RTS_World))
   > Set Pelvis Translation(Pelvis Translation:Vector=@n4) -> (Output_Get:Vector)
   > "Set Pelvis  Rotation"("Pelvis  Rotation":Rotator=@n4.Rotation) -> (Output_Get:Rotator)
   n5 = BreakTransform(InTransform=GetSocketTransform(self=@n3, InSocketName=spine_03, TransformSpace=RTS_World))
   > Set Spine_03 Translation(Spine_03 Translation:Vector=@n5) -> (Output_Get:Vector)
   > n6 = Set Spine_03 Rotation(Spine_03 Rotation:Rotator=@n5.Rotation) -> (Output_Get:Rotator)
   n7 = Cast To PlayerPawn_Athena(execute=@n18.then, Object=TryGetPawnOwner())
   n8 = Set SkelMesh(execute=@n19.then, SkelMesh=GetSkeletalMeshForPartType(self=@n7))
   Set NewVar(
    execute=@n6.then
    NewVar:Vector=MakeVector(Y=GetAlphaFromReferenceTransformDelta(SkeletalMeshComponent=@n8, BoneName=spine_03, BaseName=pelvis, DeltaComponentType=EulerY, InputRange=(X=0,Y=1), TwistAxis="1,0,0").RawOutput)
   ) -> (Output_Get:Vector)
   > n9 = Calculate Velocity From Position History!(
    execute
    DeltaSeconds:float=@n1.DeltaTimeX
    Position:Vector=Get NewVar() -> (NewVar:Vector)
    History:"Position History Structure (by ref)"=Get History() -> (History:Position History Structure)
    NumberOfSamples:int=16
    VelocityMin:float=5
    VelocityMax:float=64
   ) -> (then, ReturnValue:float)
   n10 = Get bIsPlayingMeleeAnim()
   n11 = Get bIsJumping()
   n12 = Get bIsFalling()
   n13 = Get bIsSprinting()
   n14 = "To Float (Boolean)"(InBool=NOT Boolean(A=OR Boolean!(A=@n11, B=@n12, C:bool=@n13) -> (ReturnValue)))
   > Set AlphaRotationHistory(AlphaRotationHistory:double="Clamp (Float)"(Value=float * float(A=@n9, B=@n14) {is=title K2Node_PromotableOperator}, Min=0, Max=SelectFloat(A=0.8, B=1, bPickA=@n10))) -> (Output_Get:double)
   > n15 = Gravity Override State Function DEPRECATED(CharacterPartAnimInstance=Self-Reference(), GameTimeInSeconds=GetGameTimeInSeconds(), GravityOverrideParams=Get Gravity Override Params())
   > Set Gravity Override(Gravity Override=@n15)
   > n16 = Branch(Condition=OR Boolean!(A=Get bIsCrouching(), B=Get bIsSkyDiving()) -> (ReturnValue))
   > Set Component Linear Acc Scale_2(Component Linear Acc Scale_2:Vector="0, 0, 0") -> (Output_Get:Vector)
   > Set Component Linear Vel Scale_2(Component Linear Vel Scale_2:Vector="0, 0, 0") -> (Output_Get:Vector)
   Set Component Linear Acc Scale_2(execute=@n16.else, Component Linear Acc Scale_2="0.6,0.6,0.4")
   > Set Component Linear Vel Scale_2(Component Linear Vel Scale_2="0.6,0.6,0.4")
   n17 = Branch(execute=@n2.then_1, Condition=Get bIsInFrontEnd())
   > Set Alpha_0(Alpha_0:float=0.8) -> (Output_Get:float)
   Set Alpha_0(execute=@n17.else, Alpha_0=1)
   n18 = Event OnCharacterCustomizationLoaded()
   TryGetPawnOwner()
   n19 = Set As Player Pawn Athena(execute=@n7.then, As Player Pawn Athena:Player Pawn Athena Object Reference=@n7) -> (Output_Get:Player Pawn Athena Object Reference)
  ]}
]
variables [
 Alpha {type=number, default=1}
 Alpha_0 {type=number, default=1}
 AlphaRotationHistory {type=number}
 As Player Pawn Athena {type=PlayerPawn_Athena_C}
 Component Linear Acc Scale {type=Vector, default=<0.3 0.3 0.3>}
 Component Linear Acc Scale_0 {type=Vector, default=<0.5 0.3 0.2>}
 Component Linear Acc Scale_2 {type=Vector, default=<0.5 0.5 0.2>}
 Component Linear Vel Scale {type=Vector, default=<0.3 0.3 0.3>}
 Component Linear Vel Scale_0 {type=Vector, default=<0.5 0.3 0.2>}
 Component Linear Vel Scale_2 {type=Vector, default=<0.5 0.5 0.2>}
 External Force {type=Vector}
 Gravity Override {type=Vector, default=<0 0 -700>}
 Gravity Override Params {type=GravityOverrideParamsStruct}
 History {type=PositionHistory, default {range=0, positions []}}
 NewVar {type=Vector}
 "Pelvis  Rotation" {type=Rotator}
 Pelvis Translation {type=Vector}
 SkelMesh {type=SkeletalMeshComponent}
 Spine_03 Rotation {type=Rotator}
 Spine_03 Translation {type=Vector}
]
